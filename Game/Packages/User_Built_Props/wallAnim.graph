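# Material function for wall animation on user-built props that turns a gradient and a phase value into a scalar mask.
# It negates AnimGrad, adds AnimPhase offset by one, then clamps and inverts the sum.
# The result equals AnimGrad minus AnimPhase clamped to the default range, so raising the phase drives the mask toward zero, lowest gradient values first.
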MaterialFunction /Game/Packages/User_Built_Props/wallAnim
graph {nodes [
  n1 = Add(
   A=Multiply(A=FunctionInput() {inputName=AnimGrad, inputType=FunctionInput_Scalar}) {constB=-1}
   B=Add(A=FunctionInput() {inputName=AnimPhase, inputType=FunctionInput_Scalar}, B=Constant() {r=1})
  )
  FunctionOutput(None=OneMinus(None=Clamp(None=@n1)))
 ]}
